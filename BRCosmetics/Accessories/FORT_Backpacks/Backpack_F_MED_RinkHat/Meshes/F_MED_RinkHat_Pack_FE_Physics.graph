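# Physics asset for the F_MED_RinkHat backpack cosmetic, simulating a small limbed figure built on dyn_ bones.
# Seven capsule bodies cover a spine, two-segment arms (upper and lower hand) and two thighs.
# Six constraints form a tree rooted at dyn_Spine, all with zero-degree twist limits.
# Thighs get wide swing limits, while the arm joints are much tighter, nearly rigid at the spine, with slight left/right asymmetry.
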PhysicsAsset /BRCosmetics/Accessories/FORT_Backpacks/Backpack_F_MED_RinkHat/Meshes/F_MED_RinkHat_Pack_FE_Physics
bodies {
 dyn_Spine [{shapeType=Capsule, length=5, radius=8, shapeName=dyn_Spine_capsule, rotation=<rot 0 -90 0>}]
 dyn_lowerhand_l [{shapeType=Capsule, length=5, radius=3, shapeName=dyn_lowerhand_l_capsule, center=<2 0 0>, rotation=<rot 90 0 0>}]
 dyn_lowerhand_r [{shapeType=Capsule, length=2, radius=2, shapeName=dyn_lowerhand_r_capsule, center=<4 0 0>, rotation=<rot 90 0 0>}]
 dyn_thigh_l [{shapeType=Capsule, length=5, radius=3, shapeName=dyn_thigh_l_capsule, center=<4 0 0>, rotation=<rot 90 0 0>}]
 dyn_thigh_r [{shapeType=Capsule, length=5, radius=3, shapeName=dyn_thigh_r_capsule, center=<4 0 0>, rotation=<rot 90 0 0>}]
 dyn_upperhand_l [{shapeType=Capsule, length=5, radius=3, shapeName=dyn_upperhand_l_capsule, center=<2 0 0>, rotation=<rot 90 0 0>}]
 dyn_upperhand_r [{shapeType=Capsule, length=2, radius=2, shapeName=dyn_upperhand_r_capsule, center=<1 0 0>, rotation=<rot 90 0 0>}]
}
constraints [
 {bone1Name=dyn_lowerhand_r, bone2Name=dyn_upperhand_r, swing1LimitDegrees=10, swing2LimitDegrees=10, twistLimitDegrees=0}
 {bone1Name=dyn_upperhand_r, bone2Name=dyn_Spine, swing1LimitDegrees=2, swing2LimitDegrees=2, twistLimitDegrees=0}
 {bone1Name=dyn_thigh_l, bone2Name=dyn_Spine, swing1LimitDegrees=40, swing2LimitDegrees=40, twistLimitDegrees=0}
 {bone1Name=dyn_thigh_r, bone2Name=dyn_Spine, swing1LimitDegrees=40, swing2LimitDegrees=40, twistLimitDegrees=0}
 {bone1Name=dyn_upperhand_l, bone2Name=dyn_Spine, swing1LimitDegrees=1, swing2LimitDegrees=1, twistLimitDegrees=0}
 {bone1Name=dyn_lowerhand_l, bone2Name=dyn_upperhand_l, swing1LimitDegrees=10, swing2LimitDegrees=3, twistLimitDegrees=0}
]
