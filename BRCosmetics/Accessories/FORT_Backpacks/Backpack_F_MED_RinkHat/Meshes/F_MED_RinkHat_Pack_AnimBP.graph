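# Animation Blueprint for the RinkHat backpack cosmetic, adding physics-driven secondary motion to the pack mesh.
# The input pose gets a small additive offset on the pack bone, then passes through two RigidBody nodes in root-bone space.
# bIsInFrontEnd toggles between the gameplay and front-end physics assets.
# Each update stores parameter-asset gravity and rigid-body settings in variables, while both nodes show fixed values for these inputs.
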AnimBlueprint /BRCosmetics/Accessories/FORT_Backpacks/Backpack_F_MED_RinkHat/Meshes/F_MED_RinkHat_Pack_AnimBP
target_skeleton=&/F_MED_RinkHat_Pack_Skeleton
graphs [
 AnimGraph {nodes [
   n1 = "Transform (Modify) Bone - Bone: pack"(ComponentPose=Local To Component(LocalPose=Input Pose()), Translation="0,1.5,0", Rotation="0,0,0") {
    rotationMode=BMM_Additive
    rotationSpace=BCS_BoneSpace
    translationMode=BMM_Additive
    translationSpace=BCS_BoneSpace
    boneToModify {boneName=pack}
    translation {y=1.5}
   }
   n2 = RigidBody!(
    ComponentPose=@n1
    bAlphaBoolEnabled:bool=NOT Boolean(A=Get bIsInFrontEnd())
    OverrideWorldGravity:Vector="0,0,-980"
    SimSpaceSettings:Sim Space Settings Structure=(WorldAlpha=0.5,VelocityScaleZ=0.3,DampingAlpha=1,MaxLinearVelocity=500,MaxAngularVelocity=3,MaxLinearAcceleration=300,MaxAngularAcceleration=10000,ExternalLinearDragV=(X=0,Y=0,Z=0),ExternalLinearVelocity=(X=0,Y=0,Z=0),ExternalAngularVelocity=(X=0,Y=0,Z=0))
   ) -> (Pose) {
    alphaInputType=Bool
    bOverrideWorldGravity=true
    overridePhysicsAsset=&/F_MED_RinkHat_Pack_Physics
    simulationSpace=BaseBoneSpace
    alphaBoolBlend {blendOption=Linear}
    alphaScaleBiasClamp {bInterpResult=true}
    baseBoneRef {boneName=root}
    componentAppliedLinearAccClamp=<0 0 0>
    overrideWorldGravity {z=-980}
    simSpaceSettings {maxAngularVelocity=3, maxLinearAcceleration=300, maxLinearVelocity=500, velocityScaleZ=0.3, worldAlpha=0.5}
   }
   n3 = RigidBody(
    ComponentPose=@n2
    bAlphaBoolEnabled=Get bIsInFrontEnd()
    OverrideWorldGravity="0,0,-980"
    SimSpaceSettings=(WorldAlpha=0.5,VelocityScaleZ=0.3,DampingAlpha=1,MaxLinearVelocity=500,MaxAngularVelocity=3,MaxLinearAcceleration=300,MaxAngularAcceleration=10000,ExternalLinearDragV=(X=0,Y=0,Z=0),ExternalLinearVelocity=(X=0,Y=0,Z=0),ExternalAngularVelocity=(X=0,Y=0,Z=0))
   ) {
    alphaInputType=Bool
    bOverrideWorldGravity=true
    overridePhysicsAsset=&/F_MED_RinkHat_Pack_FE_Physics
    simulationSpace=BaseBoneSpace
    alphaBoolBlend {blendOption=Linear}
    alphaScaleBiasClamp {bInterpResult=true}
    baseBoneRef {boneName=root}
    componentAppliedLinearAccClamp=<0 0 0>
    overrideWorldGravity {z=-980}
    simSpaceSettings {maxAngularVelocity=3, maxLinearAcceleration=300, maxLinearVelocity=500, velocityScaleZ=0.3, worldAlpha=0.5}
   }
   Output Pose(Result=Component To Local(ComponentPose=@n3)) {layerGroup=None}
  ]}
 EventGraph {nodes [
   Event BlueprintUpdateAnimation()
   > Sequence()
   > n1 = "Gravity Override State Function (Parameter Asset) (Nativized)"(
    CharacterPartAnimInstance=Self-Reference()
    GameTimeInSeconds=GetGameTimeInSeconds()
    Parameters=&/F_MED_RinkHat_Pack_Parameters
   )
   > Set GravityOverride(GravityOverride=@n1)
   > Set RBAN Settings Out(RBAN Settings Out=@n1.OutRigidBodyAnimNodeParameters)
   TryGetPawnOwner()
  ]}
]
variables [
 GravityOverride {type=Vector}
 RBAN Settings Out {type=FortRigidBodyAnimNodeParameters}
]
